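# Material function that layers a vertex-painted graffiti texture over incoming material attributes, also outputting its tinted color.
# Color and packed specular/metallic/roughness textures sample by UV or world/local-aligned projection through static switches.
# Vertex red drives a height lerp between texture alpha and the input's OpacityMask, with blue scaling contrast.
# Mobile blends only base color; an optional time-of-day emissive lights painted areas at night.
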MaterialFunction /Game/Environments/Artemis/Sets/Celebration/Materials/Master_Materials/MF_VPTexture_Graffiti
graph {nodes [
  n1 = FunctionInput() {inputType=FunctionInput_MaterialAttributes}
  n2 = GetMaterialAttributes(None=@n1) {attributeGetTypes [
    #1
    #2
    #3
    #4
    #5
    #6
   ]}
  n3 = TextureObjectParameter() {group=VertexPaintTexture, parameterName=VP_Color, sortPriority=1, texture=/Game/Global/Textures/Default/Blanks/T_MidgreysRGB}
  n4 = Multiply(A=ScalarParameter() {defaultValue=1, group=VertexPaintTexture, parameterName=VP_TextureScaleWS, sortPriority=4}, B=Constant() {r=256})
  n5 = VectorParameter() {
   group=VertexPaintTexture
   parameterName=VP_UVScaleByComponent
   sortPriority=3
   channelNames {a="", b=Normals, g=Specular, r=Diffuse}
   defaultValue=<rgba 1 1 1 0>
  }
  n6 = WorldPosition() {worldPositionShaderOffset=WPT_ExcludeAllShaderOffsets}
  n7 = StaticSwitchParameter(True=TransformPosition(None=@n6.XYZ) {transformSourceType=TRANSFORMPOSSOURCE_World}, False=@n6.XYZ) {group=VertexPaintTexture, parameterName=WS_VP_LocalTransform, sortPriority=6}
  n8 = VertexNormalWS()
  n9 = StaticSwitchParameter(True=Transform(None=@n8) {transformSourceType=TRANSFORMSOURCE_World, transformType=TRANSFORM_Local}, False=@n8) {group=VertexPaintTexture, parameterName=WS_VP_LocalTransform, sortPriority=6}
  n10 = MaterialFunctionCall(
   TextureObject=@n3
   TextureSize=Multiply(A=@n4, B=@n5.Diffuse)
   WorldPosition=@n7
   Export Float 4=StaticBool() {value=true}
   World Space Normal=@n9
  ) {materialFunction=&WorldAlignedTexture}
  n11 = VectorParameter() {
   group=VertexPaintTexture
   parameterName=VP_UVScaleByComponent
   sortPriority=3
   channelNames {a="", b=Normals, g=Specular, r=Diffuse}
   defaultValue=<rgba 1 1 1 0>
  }
  n12 = Multiply(A=TextureCoordinate(), B=ScalarParameter() {defaultValue=1, group=VertexPaintTexture, parameterName=VP_TextureScale, sortPriority=4})
  n13 = StaticSwitchParameter(
   True=StaticSwitchParameter(
    True=@n10.XYZ Texture
    False=StaticSwitchParameter(True=@n10.Z Texture, False=@n10.Z Texture) {group=VertexPaintTexture, parameterName=WS_Z, sortPriority=7}
   ) {group=VertexPaintTexture, parameterName=WS_XYZ, sortPriority=7}
   False=TextureSample(UVs=Multiply(A=@n11.Diffuse, B=@n12), Tex=@n3).RGBA
  ) {group=VertexPaintTexture, parameterName=WS_VPTexture, sortPriority=6}
  n14 = Multiply(
   A=ScalarParameter() {defaultValue=1, group=VertexPaintTexture, parameterName=VP_DiffuseMod}
   B=ComponentMask(None=@n13) {b=true, g=true, r=true}
  )
  n15 = MaterialFunctionCall!(
   Diffuse 1
   Diffuse 2
   Transistion Phase=VertexColor().R
   Height Texture 1=Multiply(
    A=ComponentMask(None=@n13) {a=true}
    B=ScalarParameter() {defaultValue=1, group=VertexPaintTexture, parameterName=VP_TextureHeightMax, sortPriority=8}
   )
   Height Texture 2=Multiply(
    A=GetMaterialAttributes(None=@n1) {attributeGetTypes [#7]}.OpacityMask
    B=ScalarParameter() {defaultValue=1, group=VertexPaintTexture, parameterName=VP_Background_HM_Max, sortPriority=9}
   )
   Contrast=Multiply(A=ScalarParameter() {defaultValue=5, group=VertexPaintTexture, parameterName=VP_Contrast, sortPriority=11}, B=VertexColor().B)
  ) {materialFunction=/Engine/Functions/Engine_MaterialFunctions02/Texturing/HeightLerpWithTwoHeightMaps}
  n16 = LinearInterpolate(A=@n2.BaseColor, B=@n14, Alpha=@n15.Alpha)
  n17 = TextureObjectParameter() {
   group=VertexPaintTexture
   parameterName=VP_RGBMask
   samplerType=SAMPLERTYPE_Masks
   sortPriority=2
   texture=&T_Fortnite_Default_S
  }
  n18 = MaterialFunctionCall(TextureObject=@n17, TextureSize=Multiply(A=@n4, B=@n5.Specular), WorldPosition=@n7, World Space Normal=@n9) {materialFunction=&WorldAlignedTexture}
  n19 = StaticSwitchParameter(True=@n18.Z Texture, False=@n18.Z Texture) {group=VertexPaintTexture, parameterName=WS_Z, sortPriority=7}
  n20 = Multiply(A=@n11.Specular, B=@n12)
  n21 = MaterialFunctionCall(Float3=StaticSwitchParameter(
    True=StaticSwitchParameter(True=@n18.XYZ Texture, False=@n19) {group=VertexPaintTexture, parameterName=WS_XYZ, sortPriority=7}
    False=TextureSample(UVs=@n20, Tex=@n17) {samplerType=SAMPLERTYPE_Masks}.RGB
   ) {group=VertexPaintTexture, parameterName=WS_VPTexture, sortPriority=6}) {materialFunction=&BreakOutFloat3Components}
  n22 = SetMaterialAttributes!(
   MaterialAttributes=@n1
   BaseColor=@n16
   Metallic=LinearInterpolate(A=@n2.Metallic, B=@n21.G, Alpha=@n15.Alpha) {constB=0}
   Specular=LinearInterpolate(A=@n2.Specular, B=@n21.R, Alpha=@n15.Alpha)
   Roughness=LinearInterpolate(A=@n2.Roughness, B=@n21.B, Alpha=@n15.Alpha)
  ) {attributeSetTypes [
    #1
    #5
    #2
    #3
   ]}
  n23 = SetMaterialAttributes!(MaterialAttributes=@n1, BaseColor=@n16) {attributeSetTypes [#1]}
  n24 = StaticSwitchParameter(True=@n22, False=@n23) {defaultValue=true, group=VertexPaintTexture, parameterName=Use Spec for VP?}
  n25 = ShadingPathSwitch(Default=@n24, Deferred=@n24, Mobile=@n23) {desc=Migrated from FeatureLevelSwitch}
  n26 = GetMaterialAttributes(None=@n25) {attributeGetTypes [#1]}
  n27 = Constant3Vector()
  n28 = VectorParameter() {group=VertexPaintTexture, parameterName="Emissive Color (Night)", defaultValue=<rgba 1 1 1 1>}
  n29 = Multiply(
   A=StaticSwitchParameter(True=Multiply(A=@n15.Alpha, B=@n26.BaseColor), False=@n15.Alpha) {defaultValue=true, parameterName=UseDiffuseColorForEmissive}
   B=MaterialFunctionCall(Color_Morning=@n27, Color_Day=@n27, Color_Evening=@n27, Color_Night=@n28.RGB) {materialFunction=&Color_Time_of_Day}.Result
  ) {desc=Ensuring that only VC areas are emissive}
  Add(A=@n29)
  FunctionOutput(None=StaticSwitchParameter(
    True=SetMaterialAttributes!(MaterialAttributes=@n25, EmissiveColor=@n29) {attributeSetTypes [#6]}
    False=@n25
   ) {group=VertexPaintTexture, parameterName=Use TOD Emissive?})
  FunctionOutput(None=@n14) {outputName=VP_Color, sortPriority=2}
  Multiply(B=@n28.RGB) {desc=Multiplying VC emissive by VC base color}
 ]}
