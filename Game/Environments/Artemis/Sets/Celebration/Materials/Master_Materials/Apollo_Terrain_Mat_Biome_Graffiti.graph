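# Master material for Apollo terrain meshes in the Celebration set, combining biome layers with a graffiti texture function.
# Vertex-color-blended or custom layers support optional height-lerped rock edges, masked paint, panning rice water and water-level normal flattening.
# On higher quality levels, landscape RVT blends in by height difference and vertex green, skipped for HLOD proxies.
# That result feeds graffiti, snow progression, optional Curie grass and Creative move-tool functions.
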Material /Game/Environments/Artemis/Sets/Celebration/Materials/Master_Materials/Apollo_Terrain_Mat_Biome_Graffiti
graph {
 properties {
  bTangentSpaceNormal=false
  bUseMaterialAttributes=true
  bUsedWithInstancedStaticMeshes=true
  bUsedWithNanite=true
  bUsedWithSplineMeshes=true
  physMaterial=/Game/FortressPhysicalMaterials/Generic
  pixelDepthOffsetMode=PDOM_Legacy
  usedShadingModels=Default Lit
 }
 nodes [
  Constant() {r=60}
  Constant() {r=6}
  Constant()
  n1 = StaticBool()
  n2 = DotProduct(A=VertexNormalWS(), B=Constant3Vector() {constant=<rgba 0 0 1 0>})
  n3 = RuntimeVirtualTextureSampleParameter!(
   UVs
   World Position
   Mip Bias=LinearInterpolate(
    A=ScalarParameter() {defaultValue=6, group=RVT Base, parameterName=MipBiasVerticalSurfaces, sortPriority=2}
    B=ScalarParameter() {group=RVT Base, parameterName=MipBiasHorizontalSurfaces, sortPriority=3}
    Alpha=Saturate(None=@n2)
   )
  ) {
   materialType=BaseColor_Normal_Specular_YCoCg
   mipValueMode=RVTMVM_MipBias
   parameterName=RVT Landscape Target
   virtualTexture=/Game/Athena/Apollo/Maps/Landscape/VT/RVT_Apollo_Landscape
  }
  n4 = SetMaterialAttributes!(MaterialAttributes, Base Color=@n3.BaseColor, Specular=@n3.BaseColor, Roughness=@n3.BaseColor, Normal=@n3.BaseColor) {attributeSetTypes [
    #1
    #2
    #3
    #4
   ]}
  n5 = RuntimeVirtualTextureSampleParameter!(UVs, World Position) {
   materialType=WorldHeight
   parameterName=RVT Landscape Height Target
   virtualTexture=/Game/Athena/Apollo/Maps/Landscape/VT/RVT_Apollo_Landscape_Height
  }
  n6 = ComponentMask(None=WorldPosition().XYZ) {b=true}
  n7 = Subtract(A=@n5.WorldHeight, B=@n6)
  n8 = OneMinus(None=Ceil(None=Saturate(None=@n7)))
  n9 = ScalarParameter() {defaultValue=60, group=RVT Base, parameterName=BlendwithTerrainDistance, sliderMax=1000, sliderMin=0.001, sortPriority=1}
  n10 = Divide(A=Distance(A=@n5.WorldHeight, B=@n6), B=Max(B=@n9) {constA=0.001})
  n11 = ScalarParameter() {defaultValue=1, desc=Currently uses Green channel, group=RVT Base, parameterName=UseVertColorToMaskOutRVT, sliderMax=1, sortPriority=1}
  n12 = Add(A=Saturate(None=Multiply(A=@n8, B=@n10)), B=Multiply(A=VertexColor().G, B=@n11))
  n13 = NamedRerouteUsage()
  n14 = StaticSwitchParameter(True=BlendMaterialAttributes(A=@n4, B=@n13, Alpha=Saturate(None=@n12)) {vertexAttributeBlendType=UseA}, False=@n13) {defaultValue=true, desc=Adding to allow Creative to make MIC with RVT disabled, group=RVT Base, parameterName=RVT Enable, sortPriority=50}
  n15 = QualitySwitch(
   Default=MaterialFunctionCall(NoRVT=@n13, RVT=@n14) {materialFunction=&FN_RVT_Switch_Material}.Result
   Low=@n13
   Medium=@n13
  )
  n16 = GetMaterialAttributes(None=MaterialFunctionCall!(In=MaterialProxyReplace(Realtime=@n15, MaterialProxy=@n13) {desc=Skip RVT on HLODs}) {materialFunction=&/MF_VPTexture_Graffiti}.Result) {attributeGetTypes [
    #1
    #2
    #3
    #4
   ]}
  n17 = SetMaterialAttributes(Base Color=@n16.Base Color, Specular=@n16.Base Color, Roughness=@n16.Base Color, Normal=@n16.Base Color) {attributeSetTypes [
    #1
    #2
    #3
    #4
   ]}
  n18 = MaterialFunctionCall!(
   IsStaticSnow
   UseAlternateStaticSnowMask
   AlternateStaticSnowMask
   "UseRoad&&OcclusionMask"
   IsForGrassMeshToPushDown
   AdditionalMaskToRemoveSnow
   IsFallingSnowFX
   UseMinimapCoords
   IncreaseSnowUsingSlope=@n1
   UseSmoothBorder
   LerpHighM
   TemporalHeightLerpContrast
   IsLobbyIsland
   UseBR-IslandMask=@n1
   IsTree
   IsForRenderTarget
   UseSparkle
   UseSnowNormals=StaticBool() {value=true}
   UseAdditionalMaskToRemoveSnow
   IsTangentSpaceInput
   UseVertexNormalsForBasicSnow
   UseTangentSnowNormals=@n1
   Material=@n17
  ) {materialFunction=/Game/Athena/Apollo/Environments/Landscape/Materials/Biome/UniqueBiomeMaterialFunctions/MF_Apollo_Snow_Progression}
  n19 = StaticSwitchParameter(
   True=MaterialFunctionCall!(
    In=@n18.Material
    UseHeightFalloff
    OnFireEmissiveOverride
    HeightFalloffRadius
    HeightFalloffHardness
    VariationMasksMobile
    In_Grass
    Is_Stylised_WE
   ) {materialFunction=/Game/Packages/Fortress_SharedMaterials/Base_Material_Functions/MF_Curie_Landscape}.Result
   False=@n18.Material
  ) {
   defaultValue=true
   desc="Set to true to spawn grass on this mesh.\r\nUses the vertex color to determine where to spawn."
   group=GrassSpawner
   parameterName="Use Grass(true) or a BaseTilingTexture(False)"
   sortPriority=0
  }
  MaterialFunctionCall(MA In=@n19) {materialFunction=&MF_CreativeMoveTool}
  n20 = NamedRerouteUsage()
  n21 = SetMaterialAttributes(
   Base Color=VectorParameter() {group=PaintLayer, parameterName=PaintLayerColor, defaultValue=<rgba 1 1 1 1>}.RGB
   Specular=Constant() {r=0.5}
   Roughness=ScalarParameter() {defaultValue=0.25, group=PaintLayer, parameterName=Paint Roughness}
   Normal=VertexNormalWS()
  ) {attributeSetTypes [
    #1
    #2
    #3
    #4
   ]}
  n22 = TextureCoordinate()
  n23 = VectorParameter() {group=PaintLayer, parameterName=PaintLayerUV_RG_Offset_BA_Scale, defaultValue=<rgba 0 0 1 1>}
  n24 = ComponentMask(None=@n23.RGB) {g=true, r=true}
  n25 = TextureSampleParameter2D(UVs=Multiply(A=Add(A=@n22, B=@n24), B=AppendVector(A=@n23.B, B=@n23.B))) {
   group=PaintLayer
   parameterName=PaintLayerMasks
   samplerType=SAMPLERTYPE_Masks
   texture=/Game/Packages/DS_Fortnight/SM/Textures/T_SoccerTennisBasketball
  }
  n26 = StaticComponentMaskParameter(None=AppendVector(A=@n25.RGB, B=@n25.RGB)) {defaultR=true, group=PaintLayer, parameterName=PaintLayerMaskChannel}
  n27 = WorldPosition()
  n28 = TextureSampleParameter2D(UVs=Divide(
    A=ComponentMask(None=@n27.XYZ) {g=true, r=true}
    B=ScalarParameter() {defaultValue=2048, group=PaintLayer, parameterName=PaintLayerMask_Mask Tiling}
   ) {constB=1024}) {
   group=PaintLayer
   parameterName=PaintLayerMasks_Mask
   samplerType=SAMPLERTYPE_Masks
   texture=/Game/Packages/Fortnite_UniqueMaterials_1/Textures/T_FRT_Conc_Masks_B_new
  }
  n29 = AppendVector(A=@n28.RGB, B=@n28.RGB)
  n30 = StaticSwitchParameter(
   True=Multiply(A=@n26, B=StaticComponentMaskParameter(None=@n29) {defaultR=true, group=PaintLayer, parameterName=PaintLayerMask_Mask_Channel})
   False=@n26
  ) {desc=Diffuse, group=PaintLayer, parameterName=Use Second PaintLayer Mask}
  n31 = Multiply(A=@n30, B=ScalarParameter() {defaultValue=1, group=PaintLayer, parameterName=PaintLayer Alpha Modulation})
  n32 = StaticSwitchParameter(
   True=BlendMaterialAttributes(A=@n20, B=@n21, Alpha=StaticSwitchParameter(True=@n31, False=@n30) {desc=Diffuse, group=PaintLayer, parameterName=Modulate Paint Layer Mask})
   False=@n20
  ) {desc=Uses a texture mask to paint a flat color on top of all other layers, group=Layers, parameterName=Use PaintLayer, sortPriority=100}
  NamedRerouteDeclaration(None=StaticSwitchParameter(True=NamedRerouteUsage(), False=NamedRerouteDeclaration(None=@n32) {name=Layers After Switch, nodeColor=<rgba 0 0.6953 1 1>}) {desc="R is the mask ", group=RiceWater, parameterName=RiceWater, sortPriority=-1}) {name=PaintLayer, nodeColor=<rgba 1 0 0.9844 1>}
  n33 = GetMaterialAttributes(None=NamedRerouteUsage()) {attributeGetTypes [#1]}
  n34 = ScalarParameter() {group=RiceWater, parameterName=WaterDepth Angle, sortPriority=3}
  n35 = Multiply(A=TextureCoordinate(), B=ScalarParameter() {defaultValue=5, group=RiceWater, parameterName=RiceWaterScale, sortPriority=2})
  n36 = ScalarParameter() {defaultValue=0.05, group=RiceWater, parameterName=RiceWaterSpeed}
  n37 = Panner(
   Coordinate=Multiply(A=@n35) {constB=2}
   Speed=ScalarParameter() {defaultValue=-0.05, group=RiceWater, parameterName=RiceWaterSpeed2, sortPriority=4}
  )
  n38 = Add(
   A=TextureSample(UVs=Panner(Coordinate=@n35, Speed=@n36)) {samplerType=SAMPLERTYPE_Normal, texture=/Game/Global/Textures/Default/Noise/T_SwirlNoise_N}.RGB
   B=Multiply(
    A=TextureSample(UVs=@n37) {samplerType=SAMPLERTYPE_Normal, texture=&T_SwirlNoise_N}.RGB
    B=Constant3Vector() {constant=<rgba 1 1 0 0>}
   )
  )
  n39 = LinearInterpolate(
   A=ScalarParameter() {defaultValue=1, group=RiceWater, parameterName=WaterDepth Face, sortPriority=3}
   B=@n34
   Alpha=ComponentMask(None=@n38) {r=true}
  )
  n40 = Fresnel() {baseReflectFraction=0, exponent=3}
  n41 = ScalarParameter() {defaultValue=0.01, group=RiceWater, parameterName=RiceWaterR, sortPriority=1}
  n42 = ScalarParameter() {defaultValue=1, group=RiceWater, parameterName=RiceWaterSpec, sortPriority=1}
  n43 = SetMaterialAttributes!(
   MaterialAttributes=@n33.MaterialAttributes
   Base Color=Multiply(A=@n33.Base Color, B=LinearInterpolate(A=@n39, B=@n34, Alpha=@n40)) {constB=0.25}
   Roughness=@n41
   Specular=@n42
   Normal=Multiply(A=@n38, B=Constant3Vector() {constant=<rgba 0.035 0.035 1 0>})
  ) {attributeSetTypes [
    #1
    #3
    #2
    #4
   ]}
  n44 = SetMaterialAttributes(
   MaterialAttributes=@n33.MaterialAttributes
   Base Color=Multiply(A=@n33.Base Color, B=@n34)
   Roughness=@n41
   Specular=@n42
   Normal=Constant3Vector() {constant=<rgba 0 0 1 0>}
  ) {attributeSetTypes [
    #1
    #3
    #2
    #4
   ]}
  n45 = BlendMaterialAttributes(A=NamedRerouteUsage(), B=QualitySwitch(Default=@n43, Low=@n44, High=@n43, Medium=@n43, Epic=@n43), Alpha=VertexColor().R) {vertexAttributeBlendType=UseA}
  NamedRerouteDeclaration(None=@n45) {name=Rice, nodeColor=<rgba 1 0.5156 0 1>}
  n46 = MaterialFunctionCall() {materialFunction=/Game/Athena/Apollo/Environments/Terrain_Meshes/MaterialFunctions/MF_Common_GenericMaterial}
  NamedRerouteDeclaration(None=StaticSwitchParameter(
    True=@n46.Result
    False=MaterialFunctionCall!(CustomMaterial=@n46.Result, CustomMaterial_Height=@n46.Result) {materialFunction=/Game/Athena/Apollo/Environments/Terrain_Meshes/MaterialFunctions/MF_VextexColorLayerBlend}.Result
   ) {desc="If set to true, no other layers will be used.", group=Layers, parameterName=UseCustomMaterialOnly, sortPriority=-10}) {name=BlendedLayers, nodeColor=<rgba 1 0 0.9844 1>}
  n47 = Reroute()
  n48 = MaterialFunctionCall!(
   ColorTint=MaterialFunctionCall!(DarknessAmount=Constant() {r=0.2}) {materialFunction=/Game/Athena/Apollo/Environments/Landscape/Materials/Biome/CommonMaterials/MF_HeightBasedDarkenCliff}.Result
   TilingSize=ScalarParameter() {defaultValue=400, group=RockEdgeBlend, parameterName=EdgeRockTiling}
   Use_Shift_Worldposition
  ) {materialFunction=/Game/Athena/Apollo/Environments/Terrain_Meshes/MaterialFunctions/MF_EdgeRocks}
  n49 = GetMaterialAttributes(None=@n48.XY Walls) {attributeGetTypes [#1]}
  n50 = Constant3Vector() {constant=<rgba 2 0 1.785 0>}
  n51 = ScalarParameter() {defaultValue=1, group=RockEdgeBlend, parameterName=DEBUG-EdgeRock}
  n52 = ComponentMask(None=VertexNormalWS()) {b=true}
  n53 = Reroute() {desc=RVT worldheight sample}
  n54 = ComponentMask(None=WorldPosition().XYZ) {b=true}
  n55 = Ceil(None=Saturate(None=Subtract(A=@n53, B=@n54)))
  n56 = Distance(A=@n53, B=@n54)
  n57 = ScalarParameter() {defaultValue=500, group=RockEdgeBlend, parameterName=BlendwithTerrainDistance_Rocks}
  n58 = Saturate(None=Multiply(A=OneMinus(None=@n55), B=Divide(A=@n56, B=@n57)))
  n59 = Multiply(
   A=OneMinus(None=Saturate(None=@n52))
   B=Multiply(A=OneMinus(None=@n58), B=ScalarParameter() {defaultValue=0.8, group=RockEdgeBlend, parameterName=EdgeBlend_Transition2})
  )
  n60 = BlendMaterialAttributes(
   A=@n47
   B=SetMaterialAttributes!(MaterialAttributes=@n49.MaterialAttributes, Base Color=LinearInterpolate(A=@n49.Base Color, B=@n50, Alpha=@n51)) {attributeSetTypes [#1]}
   Alpha=MaterialFunctionCall(
    Transition Phase=VertexInterpolator(VS=@n59).PS
    Height Texture=@n48.XY Walls Height
    Contrast=ScalarParameter() {defaultValue=8, group=RockEdgeBlend, parameterName=EdgeBlend_Contrast}
   ) {materialFunction=&HeightLerp}.Alpha
  )
  n61 = MaterialFunctionCall(
   NoRVT=@n47
   RVT=StaticSwitchParameter(True=@n47, False=StaticSwitchParameter(True=@n60, False=@n47) {group=RockEdgeBlend, parameterName=UseRockEdgeBlend}) {desc=Replaces Cliff Material, group=Layers, parameterName=UseCustomMaterialBlend}
  ) {materialFunction=&FN_RVT_Switch_Material}
  QualitySwitch(Default=@n61.Result, Low=@n47, High=@n61.Result, Medium=@n47)
  Reroute(None=@n45)
  n62 = LinearInterpolate(
   A=LinearInterpolate(
    A=Constant3Vector() {constant=<rgba 0.0073 0.045 0.0027 1>}
    B=Constant3Vector() {constant=<rgba 0.135 0.0934 0.0614 1>}
    Alpha=VertexColor().B
   )
   B=Constant3Vector() {constant=<rgba 0.03 0.0192 0.0099 1>}
   Alpha=VertexColor().A
  )
  n63 = Constant3Vector() {constant=<rgba 0.18 0.1795 0.1791 1>}
  n64 = VertexColor()
  Reroute(None=MakeMaterialAttributes!(
    BaseColor=LinearInterpolate(A=@n62, B=@n63, Alpha=@n64.R)
    Metallic=Constant()
    Specular
    Roughness=Constant() {r=1}
    Anisotropy
    EmissiveColor
    Opacity
    OpacityMask
    Normal=VertexNormalWS()
    Tangent
    WorldPositionOffset
    SubsurfaceColor
    ClearCoat
    ClearCoatRoughness
    AmbientOcclusion
    Refraction
    CustomizedUVs_0
    CustomizedUVs_1
    CustomizedUVs_2
    CustomizedUVs_3
    CustomizedUVs_4
    CustomizedUVs_5
    CustomizedUVs_6
    CustomizedUVs_7
    PixelDepthOffset
    ShadingModel
    Displacement
   ))
  Reroute(None=@n32)
  ScalarParameter() {group=PaintLayer, parameterName=Paint Metallic}
  n65 = StaticSwitchParameter(True=Constant() {r=1}, False=Constant()) {
   defaultValue=true
   desc="Set to true to spawn grass on this mesh.\r\nUses the vertex color to determine where to spawn."
   group=GrassSpawner
   parameterName="Use Grass(true) or a BaseTilingTexture(False)"
   sortPriority=0
  }
  StaticSwitchParameter(True=@n65) {defaultValue=true, group=Layers, parameterName=Use a base Layer}
  n66 = NamedRerouteUsage()
  n67 = GetMaterialAttributes(None=@n66) {attributeGetTypes [#4]}
  n68 = Constant3Vector() {constant=<rgba 0 0 1 0>}
  n69 = ComponentMask(None=WorldPosition().XYZ) {b=true}
  n70 = StaticSwitchParameter(
   True=CollectionParameter() {collection=/Game/Athena/Apollo/Environments/Landscape/MPC/Apollo_MPC_Terrain, parameterName=WaterLevel}
   False=ScalarParameter() {defaultValue=-3900, group=FlattenNormal, parameterName=Flat_Normal_Height}
  ) {defaultValue=true, group=FlattenNormal, parameterName=UseMPCWaterLevel, sortPriority=2}
  n71 = Saturate(None=Divide(A=Subtract(A=@n69, B=@n70), B=ScalarParameter() {defaultValue=-25, group=FlattenNormal, parameterName=Flat_Normal_Blr, sortPriority=1}))
  StaticSwitchParameter(
   True=SetMaterialAttributes!(MaterialAttributes=@n67.MaterialAttributes, Normal=LinearInterpolate(A=@n67.Normal, B=@n68, Alpha=@n71)) {attributeSetTypes [#4]}
   False=@n66
  ) {group=FlattenNormal, parameterName=FlatNormal}
  WorldPosition()
 ]
}
