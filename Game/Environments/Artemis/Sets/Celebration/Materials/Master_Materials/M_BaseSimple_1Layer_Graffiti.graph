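# Single-layer master material for Celebration set environment assets, named for graffiti, built on material attributes with Default Lit shading.
# Diffuse, packed masks and normals support static-switched two-sided textures, with color-mask tinting or swatches and neon or masked emissive.
# The default path chains optional effect functions like decals, gradients, screens, fake windows and switchable moss.
# Forward and Mobile take shorter branches through ShadingPathSwitch before the game-wide shared function.
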Material /Game/Environments/Artemis/Sets/Celebration/Materials/Master_Materials/M_BaseSimple_1Layer_Graffiti
graph {
 properties {
  bAutomaticallySetUsageInEditor=false
  bUseMaterialAttributes=true
  ditheredLODTransition=true
  pixelDepthOffsetMode=PDOM_Legacy
  usedShadingModels=Default Lit
 }
 nodes [
  n1 = MaterialFunctionCall() {materialFunction=/Game/Athena/WinterArt/Materials/MF_UVsSet}
  n2 = VectorParameter() {
   group=Material_1
   parameterName=L1_UVScaleByComponent
   sortPriority=3
   channelNames {a="", b=Normals, g=Specular, r=Diffuse}
   defaultValue=<rgba 1 1 1 0>
  }
  n3 = TextureSampleParameter2D(UVs=Multiply(A=@n1.UVs Layer 1, B=@n2.Specular)) {
   group=Material_1
   parameterName=SpecularMasks Side 2
   samplerSource=SSM_Wrap_WorldGroupSettings
   samplerType=SAMPLERTYPE_Masks
   sortPriority=2
   texture=&T_Fortnite_Default_S
  }
  n4 = TwoSidedSign()
  n5 = VectorParameter() {
   group=Material_1
   parameterName=L1_UVScaleByComponent
   sortPriority=3
   channelNames {a="", b=Normals, g=Specular, r=Diffuse}
   defaultValue=<rgba 1 1 1 0>
  }
  n6 = TextureSampleParameter2D(UVs=Multiply(A=@n5.Specular, B=@n1.UVs Layer 1)) {
   group=Material_1
   parameterName=SpecularMasks
   samplerSource=SSM_Wrap_WorldGroupSettings
   samplerType=SAMPLERTYPE_Masks
   sortPriority=2
   texture=&T_Fortnite_Default_S
  }
  n7 = MaterialFunctionCall(Float4=StaticSwitchParameter(True=LinearInterpolate(A=@n3.RGBA, B=@n6.RGBA, Alpha=@n4), False=@n6.RGBA) {group=Material_1, parameterName=TwoSided, sortPriority=18}) {materialFunction=&BreakOutFloat4Components}
  n8 = TextureSampleParameter2D(UVs=Multiply(A=@n1.UVs Layer 1, B=@n2.Diffuse)) {
   group=Material_1
   parameterName=Diffuse Side 2
   samplerSource=SSM_Wrap_WorldGroupSettings
   sortPriority=1
   texture=/Game/Global/Textures/Default/Blanks/T_MidgreysRGB
  }
  n9 = Multiply(A=@n1.UVs Layer 1, B=@n5.Diffuse)
  n10 = TextureSampleParameter2D(UVs=@n9) {
   group=Material_1
   parameterName=Background Diffuse
   samplerSource=SSM_Wrap_WorldGroupSettings
   texture=&T_WhitesRGB
  }
  n11 = ScalarParameter() {defaultValue=1, group=Material_1, parameterName=L1_UV_ColorSwatchScaleAdd}
  n12 = TextureSampleParameter2D(UVs=StaticSwitchParameter(True=Multiply(A=@n11, B=@n9), False=@n9) {group=Material_1, parameterName=Use Color Swatch}) {
   group=Material_1
   parameterName=Diffuse
   samplerSource=SSM_Wrap_WorldGroupSettings
   sortPriority=1
   texture=&T_MidgreysRGB
  }
  n13 = StaticSwitchParameter(True=@n10.A, False=@n12.A) {group=Material_1, parameterName=Use Color Swatch}
  n14 = MaterialFunctionCall!(
   Transition=ScalarParameter() {defaultValue=0.5, group=Material_1, parameterName=L1_TintTransition, sortPriority=5}
   HM=@n13
   Contrast=ScalarParameter() {defaultValue=5, group=Material_1, parameterName=L1_TintContrast, sortPriority=5}
   Metalness=@n6.G
   BG Color=StaticSwitchParameter(True=@n10.RGB, False=@n12.RGB) {group=Material_1, parameterName=Use Color Swatch}
   TintColor=StaticSwitchParameter(True=@n12.RGB, False=VectorParameter() {group=Material_1, parameterName=TintColor, sortPriority=7, defaultValue=<rgba 1 1 1 0>}.RGB) {group=Material_1, parameterName=Use Color Swatch}
  ) {materialFunction=/Game/Athena/WinterArt/Materials/MF_ColorMask}
  n15 = AppendVector(
   A=StaticSwitchParameter(True=@n14.LerpMix, False=@n14.LerpMix) {group=Material_1, parameterName=L1_UseTintColorAsDiffuse}
   B=Multiply(A=@n13, B=ScalarParameter() {defaultValue=1, group=Material_1, parameterName=L1 Alpha, sortPriority=1})
  )
  n16 = StaticSwitchParameter(True=LinearInterpolate(A=@n8.RGBA, B=@n15, Alpha=@n4), False=@n15) {group=Material_1, parameterName=TwoSided, sortPriority=18}
  n17 = ComponentMask(None=@n16) {b=true, g=true, r=true}
  n18 = MaterialFunctionCall!(Base Diffuse=TextureObjectParameter() {group=Material_1, parameterName=Diffuse, sortPriority=1, texture=&T_MidgreysRGB}) {materialFunction=/Game/Athena/WinterArt/Materials/MF_NeonSignEmissive}
  n19 = VectorParameter() {group=Material_1, parameterName=Emissive, sortPriority=7, defaultValue=<rgba 0 0 0 1>}
  n20 = MaterialFunctionCall!(
   X=ScalarParameter() {group=Adv Emissive, parameterName=L1 X Speed, sortPriority=12}
   Y=ScalarParameter() {group=Adv Emissive, parameterName=L1 Y Speed, sortPriority=13}
   Dot=VectorParameter() {group=Adv Emissive, parameterName=L1 Channel, sortPriority=9}.RGB
   TextureOffset=ScalarParameter() {defaultValue=1, group=Adv Emissive, parameterName=L1 Offset, sliderMax=4, sliderMin=1, sortPriority=11}
   EmissiveTexture=Multiply(A=@n19.RGB, B=@n19.RGB)
  ) {materialFunction=/Game/Athena/WinterArt/Materials/MF_EmissiveMask}
  n21 = ComponentMask(None=@n16) {a=true}
  n22 = ScalarParameter() {defaultValue=1, group=SSS, parameterName=SSSVal}
  n23 = TextureSampleParameter2D(UVs=Multiply(A=@n1.UVs Layer 1, B=@n2.Normals)) {
   group=Material_1
   parameterName=Normals Side 2
   samplerSource=SSM_Wrap_WorldGroupSettings
   samplerType=SAMPLERTYPE_Normal
   sortPriority=3
   texture=/Engine/EngineMaterials/FlatNormal
  }
  n24 = TextureSampleParameter2D(UVs=Multiply(A=@n5.Normals, B=@n1.UVs Layer 1)) {
   group=Material_1
   parameterName=Normals
   samplerSource=SSM_Wrap_WorldGroupSettings
   samplerType=SAMPLERTYPE_Normal
   sortPriority=3
   texture=&FlatNormal
  }
  n25 = SetMaterialAttributes!(
   MaterialAttributes
   Base Color=@n16
   Specular=@n7.R
   Metallic=@n7.R
   Roughness=@n7.R
   Emissive Color=MaterialFunctionCall!(
    BaseColor=@n17
    EmissiveMask=@n7.A
    EmissiveColor=StaticSwitchParameter(True=@n18.Result, False=@n20.EmissiveMasked) {group=NeonEmissive, parameterName=UseNeonSignEmissive_Layer1}
   ) {materialFunction=/Game/Athena/WinterArt/Materials/MF_EmissiveControls}.Result
   Normal=StaticSwitchParameter(True=LinearInterpolate(A=@n23.RGB, B=@n24.RGB, Alpha=@n4), False=@n24.RGB) {group=Material_1, parameterName=TwoSided, sortPriority=18}
   Custom Data 0=Multiply(A=ScalarParameter() {group=Material_1, parameterName=L1 Clearcoat}, B=@n21)
   Custom Data 1=Constant() {r=0.025}
   Opacity Mask=@n21
   Subsurface Color=Multiply(A=@n21, B=Multiply(A=@n17, B=@n22))
   Opacity=ScalarParameter() {defaultValue=0.5, group=SSS, parameterName=SSSDepth}
  ) {attributeSetTypes [
    #1
    #2
    #3
    #4
    #5
    #6
    #7
    #8
    #9
    #10
    #11
   ]}
  n26 = MaterialFunctionCall!(In=MaterialFunctionCall!(In=@n25, UseOneSampler) {materialFunction=/Game/Athena/WinterArt/Materials/MF_VPTexture}.Result) {materialFunction=/Game/Athena/WinterArt/Materials/MF_DecalPrint}
  n27 = MaterialFunctionCall!(In=MaterialFunctionCall!(In=@n26.Result) {materialFunction=/Game/Athena/WinterArt/Materials/MF_SphereEmissive}.Result) {materialFunction=/Game/Athena/WinterArt/Materials/MF_TOD_EmissiveColors}
  n28 = MaterialFunctionCall!(
   transitionFactor
   MaterialIn=MaterialFunctionCall!(In=MaterialFunctionCall!(In=@n27.Result) {materialFunction=/Game/Athena/WinterArt/Materials/MF_WS_Texture}.Result) {materialFunction=/Game/Athena/WinterArt/Materials/MF_Fuzzy}.Result
  ) {materialFunction=/Game/Athena/WinterArt/Materials/MF_ObjectSpaceGradient}
  n29 = MaterialFunctionCall!(In=MaterialFunctionCall!(
    In=MaterialFunctionCall!(transitionFactor, MaterialIn=@n28.Color*Gradient) {materialFunction=/Game/Athena/WinterArt/Materials/MF_WSGradient}.Color*Gradient
    Emissive Mask=@n7.A
   ) {materialFunction=/Game/Athena/WinterArt/Materials/MF_Screens}.Result) {materialFunction=/Game/Athena/WinterArt/Materials/M_FakeWindowReflections}
  n30 = MaterialFunctionCall!(WorldSpaceNormals, In=@n29.Result) {materialFunction=/Game/Environments/Apollo/Rocks/Natural_Rocks/Materials/MossSoft2}
  n31 = MaterialFunctionCall(In=@n26.Result, Emissive Mask=@n7.A) {materialFunction=&MF_Screens}
  n32 = GetMaterialAttributes(None=@n31.Result) {attributeGetTypes [#1, #5, #9]}
  n33 = SetMaterialAttributes!(
   MaterialAttributes
   Base Color=@n32.Base Color
   Emissive Color=@n32.Base Color
   Opacity Mask=@n32.Base Color
   Roughness=ScalarParameter() {defaultValue=0.5, group=Mobile, parameterName=Mobile_Roughness}
   Specular=ScalarParameter() {defaultValue=0.5, group=Mobile, parameterName=Mobile_Specular}
  ) {attributeSetTypes [
    #1
    #5
    #9
    #4
    #2
   ]}
  MaterialFunctionCall!(In=ShadingPathSwitch(Default=StaticSwitchParameter(True=@n30.Result, False=@n29.Result) {parameterName=UseMoss}, Forward=@n31.Result, Mobile=@n33)) {materialFunction=/Game/Athena/WinterArt/Materials/MF_GameWideShared}
 ]
}
