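# Master material for the Celebration version of the Artemis HollowBase cave, covering rock faces, props and sand meshes.
# Static switches select UV or world-aligned rock and prop textures, adding details, waterline darkening and vertex-painted, height-blended sand.
# Base color is multiplied by fake cave lighting: a point light, animated caustics and hardcoded sphere-mask corner lights, scaled by quality level.
# Vertex green masks a final celebration-light tint; Fortnite hit effects drive world position offset.
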Material /Game/Environments/Artemis/Sets/Celebration/Materials/Master_Materials/M_Cave_HollowBase_Celebration
graph {
 properties {bTangentSpaceNormal=false, bUseMaterialAttributes=true, pixelDepthOffsetMode=PDOM_Legacy, usedShadingModels=Default Lit}
 nodes [
  Composite()
  Composite()
  Composite()
  n1 = VertexColor()
  n2 = MaterialFunctionCall(In=ComponentMask(None=VertexNormalWS()) {b=true}, Contrast=ScalarParameter() {defaultValue=0.4, group=Sand, parameterName=SandOfs}) {materialFunction=&CheapContrast}
  n3 = BlendMaterialAttributes(
   A=NamedRerouteUsage()
   B=NamedRerouteUsage()
   Alpha=MaterialFunctionCall(Transition Phase=Multiply(A=@n1.B, B=@n2.Result), Height Texture=NamedRerouteUsage(), Contrast=Constant() {r=0.5}) {materialFunction=&HeightLerp}.Alpha
  )
  n4 = StaticSwitchParameter(True=@n3, False=NamedRerouteUsage()) {group=RockFace, parameterName=RockFaceWithSand}
  n5 = StaticSwitchParameter(True=NamedRerouteUsage(), False=@n3) {group=RockFace, parameterName=NoSand}
  n6 = GetMaterialAttributes(None=StaticSwitchParameter(
    True=NamedRerouteUsage()
    False=StaticSwitchParameter(True=@n4, False=@n5) {desc=" Vertex A to blend ", group=RockFace, parameterName=RockFace}
   ) {group=Sand, parameterName=SandMesh}) {attributeGetTypes [#1, #2]}
  n7 = VectorParameter() {group=Lighting, parameterName=Shadow Cave Color, sortPriority=1, defaultValue=<rgba 0.3733 0.3442 0.2605 0>}
  n8 = VectorParameter() {group=Lighting, parameterName="Light  Cave Color", sortPriority=2, defaultValue=<rgba 0.4245 0.8177 0.7758 0>}
  n9 = Multiply(A=@n8.RGB, B=ScalarParameter() {defaultValue=3, group=Lighting, parameterName=LightValue})
  n10 = DotProduct(A=VertexNormalWS(), B=Constant3Vector() {constant=<rgba 0 0 -1 0>})
  n11 = Subtract(A=WorldPosition().XYZ, B=Constant3Vector() {constant=<rgba -16896 112896 0 0>})
  n12 = ComponentMask(None=@n11) {g=true, r=true}
  n13 = Constant() {r=2000}
  n14 = Multiply(A=Time(), B=Constant() {r=0.5})
  n15 = TextureSample(UVs=Rotator(Coordinate=Divide(A=@n12, B=@n13), Time=@n14) {speed=0.0075}) {samplerSource=SSM_Wrap_WorldGroupSettings, samplerType=SAMPLERTYPE_Grayscale, texture=&LowResBlurredNoise}
  n16 = ComponentMask(None=Add(A=@n11, B=Constant3Vector() {constant=<rgba 555 5555 0 0>})) {g=true, r=true}
  n17 = TextureSample(UVs=Rotator(Coordinate=Divide(A=@n16, B=@n13), Time=@n14) {speed=-0.0075}) {samplerSource=SSM_Wrap_WorldGroupSettings, samplerType=SAMPLERTYPE_Grayscale, texture=&LowResBlurredNoise}
  n18 = Multiply(
   A=NamedRerouteDeclaration(None=Saturate(None=@n10)) {name=Inverted Z, nodeColor=<rgba 0.1953 0 1 1>}
   B=Multiply(A=Constant() {r=5}, B=Multiply(A=@n15.R, B=@n17.R))
  ) {constB=0}
  MaterialExpressionReroute_1 = Reroute(None=@n18)
  n19 = Add(A=@n9, B=@MaterialExpressionReroute_1.Output)
  n20 = VectorParameter() {group=Lighting, parameterName="Sphere Center ( Fake point light)", defaultValue=<rgba 40216 -76028 4554.1543 0>}
  n21 = WorldPosition()
  n22 = NamedRerouteUsage()
  n23 = Divide(A=DotProduct(A=Subtract(A=@n20.RGB, B=@n21.XYZ), B=Transform(None=@n22)), B=Constant() {r=2000})
  n24 = Constant() {r=6000}
  n25 = Constant()
  n26 = LinearInterpolate(
   A=@n7.RGB
   B=StaticSwitchParameter(True=QualitySwitch(Default=@n9, High=@n19, Epic=@n19), False=@n9) {defaultValue=true, group=Lighting, parameterName=Caustic}
   Alpha=Multiply(A=Saturate(None=@n23), B=SphereMask(A=@n21.XYZ, B=@n20.RGB, Radius=@n24, Hardness=@n25))
  )
  n27 = VectorParameter() {group=Lighting, parameterName=CornerLight, defaultValue=<rgba 1.5146 1.7256 2 0>}
  n28 = NamedRerouteUsage()
  n29 = Add(A=NamedRerouteUsage(), B=OneMinus(None=Constant() {r=3632.0942}))
  n30 = Constant() {r=1565.0486}
  n31 = LinearInterpolate(
   A=QualitySwitch(Default=@n26, Epic=LinearInterpolate(A=@n26, B=@n27.RGB, Alpha=@n28))
   B=Constant() {r=1}
   Alpha=Saturate(None=Divide(A=@n29, B=@n30))
  )
  n32 = GetMaterialAttributes(None=SetMaterialAttributes!(MaterialAttributes=@n6.MaterialAttributes, Base Color=Multiply(A=@n6.Base Color, B=@n31)) {attributeSetTypes [#1]}) {attributeGetTypes [#2]}
  n33 = Transform(None=@n32.Normal)
  n34 = NamedRerouteUsage()
  n35 = NamedRerouteUsage()
  n36 = VertexColor()
  n37 = LinearInterpolate(A=@n33, B=LinearInterpolate(A=QualitySwitch(Default=@n34, Epic=@n35), B=@n33, Alpha=@n36.A), Alpha=@n36.B)
  n38 = TextureObject() {samplerType=SAMPLERTYPE_Normal, texture=/Game/Environments/Artemis/Sets/Celebration/Textures/Graffiti/T_Celebration_Light_D}
  n39 = ScalarParameter() {defaultValue=5000, group=Celebration, parameterName=Celebration Light tiling}
  n40 = VectorParameter() {parameterName=Lowend Celebration Ceiling Color, defaultValue=<rgba 0.6364 0.0064 1.2 1>}
  MaterialFunctionCall!(
   Material=SetMaterialAttributes!(
    MaterialAttributes=SetMaterialAttributes!(MaterialAttributes=@n32.MaterialAttributes, Normal=@n37) {attributeSetTypes [#2]}
    World Position Offset=NamedRerouteUsage()
   ) {attributeSetTypes [#3]}
   Tint=QualitySwitch(
    Default=MaterialFunctionCall(TextureObject=@n38, TextureSize=@n39) {materialFunction=&WorldAlignedTexture}.XYZ Texture
    Low=@n40.RGB
    Medium=@n40.RGB
   )
   TintMask=Multiply(A=ScalarParameter() {defaultValue=0.4, group=Celebration, parameterName=Celebration Light Intensity}, B=VertexColor().G)
  ) {materialFunction=/Engine/Functions/MaterialLayerFunctions/MatLayerBlend_Tint}
  n41 = WorldPosition()
  NamedRerouteDeclaration(None=Subtract(A=ComponentMask(None=@n41.XYZ) {b=true}, B=ScalarParameter() {defaultValue=1856, group=Lighting, parameterName=WaterlineZ})) {name=Waterline H, nodeColor=<rgba 0.9688 1 0 1>}
  NamedRerouteDeclaration(None=@n6.Normal) {name=Final Normal Map, nodeColor=<rgba 1 0 0.75 1>}
  n42 = TextureObjectParameter() {
   group=Sand
   parameterName=Sand Specular
   samplerType=SAMPLERTYPE_Masks
   texture=/Game/Environments/Apollo/Sets/Terrain/Textures/T_Apollo_Sand_S_01
  }
  n43 = TextureObjectParameter() {
   group=Sand
   parameterName=Sand Normal
   samplerType=SAMPLERTYPE_Normal
   texture=/Game/Environments/Apollo/Sets/Terrain/Textures/T_Apollo_Sand_N_01
  }
  n44 = MaterialFunctionCall!(
   Diffuse Texture Object=TextureObjectParameter() {group=Sand, parameterName=Sand Diffuse, texture=/Game/Environments/Apollo/Sets/Terrain/Textures/T_Apollo_Sand_D_01}
   Spec Texture Object=@n42
   TextureSize=ScalarParameter() {defaultValue=812.1213, group=Sand, parameterName=Sand WS Scale, sortPriority=1}
   --------------------------
   Use seperate settings for Z projection
   Use seperate texture for Z projection
   Texture Size Z projection
   Z Plane Diffuse Texture Object
   Z Plane Spec Texture Object
   "--------------------------      "
   ProjectionTransitionContrast
   Use Texture Alpha Heightmap to Refine the Transistion
   HeightLerpTransitionContrast
   PickTheDiffuseHeightChannel
   TheDiffuseHeightChannel
   "--------------------------       "
   NormalTextureObject=@n43
   World Space Normals=StaticBool()
   "--------------------------  "
   Export Float 4 Diffuse=StaticBool() {value=true}
   Export Float 4 Spec
   "--------------------------     "
   WorldPosition
   World Space Normal
  ) {materialFunction=/Engine/Functions/Engine_MaterialFunctions02/Texturing/WorldAlignedTextures_Complex}
  n45 = MaterialFunctionCall(Float3=@n44.Z Spec) {materialFunction=&BreakOutFloat3Components}
  NamedRerouteDeclaration(None=@n45.B) {name=Sand R, nodeColor=<rgba 0.4531 0 1 1>}
  n46 = VectorParameter() {group=Sand, parameterName=ColorSand, defaultValue=<rgba 0.5052 0.5052 0.5052 0.5>}
  n47 = Multiply(A=@n46.A, B=@n45.R)
  n48 = NamedRerouteUsage()
  NamedRerouteDeclaration(None=LinearInterpolate(B=NamedRerouteDeclaration(None=@n47) {name=Sand S, nodeColor=<rgba 0.9219 0 1 1>}, Alpha=@n48) {constA=0.2}) {name=Sand Waterline S, nodeColor=<rgba 0 0.6016 1 1>}
  NamedRerouteDeclaration(None=LinearInterpolate(B=NamedRerouteUsage(), Alpha=@n48) {constA=0.2}) {name=Sand Waterline R, nodeColor=<rgba 1 0.6328 0 1>}
  n49 = NamedRerouteDeclaration(None=Multiply(A=ComponentMask(None=@n44.Z Diffuse Texture) {b=true, g=true, r=true}, B=@n46.RGB)) {name=Sand Color, nodeColor=<rgba 0.9922 0 1 1>}
  NamedRerouteDeclaration(None=LinearInterpolate(A=Multiply(A=@n49) {constB=0.35}, B=@n49, Alpha=@n48)) {name=Sand Waterline D, nodeColor=<rgba 1 0 0.1172 1>}
  NamedRerouteDeclaration(None=ComponentMask(None=@n44.Z Diffuse Texture) {a=true}) {name=Sand A, nodeColor=<rgba 1 0.0469 0 1>}
  n50 = SetMaterialAttributes!(MaterialAttributes, Base Color=NamedRerouteUsage(), Roughness=NamedRerouteUsage(), Specular=NamedRerouteUsage(), Normal=NamedRerouteUsage()) {attributeSetTypes [
    #1
    #4
    #5
    #2
   ]}
  n51 = NamedRerouteUsage()
  n52 = SetMaterialAttributes(Base Color=NamedRerouteUsage(), Roughness=NamedRerouteUsage(), Specular=NamedRerouteUsage(), Normal=NamedRerouteUsage()) {attributeSetTypes [
    #1
    #4
    #5
    #2
   ]}
  n53 = NamedRerouteDeclaration(None=@n52) {name=Sand Layer, nodeColor=<rgba 0 1 0.3672 1>}
  NamedRerouteDeclaration(None=QualitySwitch(Default=@n50, Epic=StaticSwitchParameter(True=@n51, False=@n53) {group=Sand 2, parameterName=Layer 2 Sand})) {name=WS Sand Mat, nodeColor=<rgba 0 1 0.25 1>}
  n54 = VectorParameter() {group=Props, parameterName=PropTone, defaultValue=<rgba 1 1 1 1>}
  n55 = TextureSampleParameter2D() {
   group=Props
   parameterName=Diffuse
   samplerSource=SSM_Wrap_WorldGroupSettings
   texture=/Game/Environments/Apollo/Rocks/Natural_Rocks/Textures/T_NaturalRock_AB_D
  }
  n56 = TextureSampleParameter2D() {
   group=Props
   parameterName=SpecularMask
   samplerSource=SSM_Wrap_WorldGroupSettings
   samplerType=SAMPLERTYPE_Masks
   texture=/Game/Environments/Apollo/Rocks/Natural_Rocks/Textures/T_NaturalRock_AB_S
  }
  n57 = TextureSampleParameter2D() {
   group=Props
   parameterName=Normals
   samplerSource=SSM_Wrap_WorldGroupSettings
   samplerType=SAMPLERTYPE_Normal
   texture=/Game/Environments/Apollo/Rocks/Natural_Rocks/Textures/T_NaturalRock_AB_N
  }
  n58 = SetMaterialAttributes!(
   MaterialAttributes
   Base Color=Multiply(A=@n54.RGB, B=@n55.RGB)
   Specular=@n56.R
   Roughness=Multiply(A=@n54.A, B=@n56.B)
   Normal=@n57.RGB
   Metallic=@n56.R
  ) {attributeSetTypes [
    #1
    #5
    #4
    #2
    #6
   ]}
  n59 = StaticSwitchParameter(True=BlendMaterialAttributes(A=NamedRerouteUsage(), B=@n58, Alpha=VertexColor().G), False=@n58) {desc=" Vertex A to blend ", group=RockFace, parameterName=Prop With WS Rock}
  n60 = NamedRerouteUsage()
  n61 = TextureObject() {texture=/Game/Environments/Artemis/Sets/HollowBase/Terrain/Textures/T_Artemis_Cliff_HollowBase_01_D}
  n62 = TextureSample(Tex=@n61) {samplerSource=SSM_Wrap_WorldGroupSettings}
  n63 = TextureObject() {samplerType=SAMPLERTYPE_Normal, texture=/Game/Environments/Artemis/Sets/Terrain/Textures/T_Artemis_Cliff_01_N}
  n64 = SetMaterialAttributes!(
   MaterialAttributes
   Base Color=@n62.RGB
   Specular=Multiply(A=@n62.RGB, B=ScalarParameter() {defaultValue=0.5, group=Rocks, parameterName=Rocks Spec, sortPriority=3})
   Roughness=LinearInterpolate(A=ScalarParameter() {defaultValue=0.6393, parameterName=R1}, B=ScalarParameter() {defaultValue=2, parameterName=R2}, Alpha=@n62.R)
   Normal=TextureSample(Tex=@n63) {samplerSource=SSM_Wrap_WorldGroupSettings, samplerType=SAMPLERTYPE_Normal}.RGB
  ) {attributeSetTypes [
    #1
    #5
    #4
    #2
   ]}
  n65 = VertexColor()
  n66 = StaticSwitchParameter(True=QualitySwitch(Default=BlendMaterialAttributes(A=@n60, B=@n64, Alpha=@n65.G), Low=@n64), False=NamedRerouteUsage()) {desc=" Vertex A to blend ", group=RockFace, parameterName=RockFace}
  n67 = GetMaterialAttributes(None=NamedRerouteDeclaration(None=StaticSwitchParameter(True=@n59, False=@n66) {desc=" Vertex A to blend ", group=RockFace, parameterName=Prop Asset}) {name=Textures, nodeColor=<rgba 0.7109 1 0 1>}) {attributeGetTypes [#1, #4]}
  n68 = SetMaterialAttributes!(
   MaterialAttributes=@n67.MaterialAttributes
   Base Color=MaterialFunctionCall(Base=@n67.Base Color, Blend=Constant() {r=0.5}) {materialFunction=&Blend_Overlay}.Result
  ) {attributeSetTypes [#1]}
  n69 = MaterialFunctionCall(
   Diffuse Texture Object=TextureObject() {texture=/Game/Environments/Apollo/Rocks/Natural_Rocks/Textures/T_RockDetails_02D}
   TextureSize=ScalarParameter() {defaultValue=1111, group=Detail Texture, parameterName=WSDetailTile, sortPriority=3}
  ) {materialFunction=&WorldAlignedTextures_Complex}
  n70 = StaticSwitchParameter(
   True=SetMaterialAttributes!(
    MaterialAttributes=@n67.MaterialAttributes
    Base Color=MaterialFunctionCall(Base=@n67.Base Color, Blend=@n69.XYZ Diffuse Texture) {materialFunction=&Blend_Overlay}.Result
    Roughness=@n67.MaterialAttributes
   ) {attributeSetTypes [#1, #4]}
   False=NamedRerouteUsage()
  ) {defaultValue=true, group=Detail Texture, parameterName=Details}
  n71 = GetMaterialAttributes(None=NamedRerouteDeclaration(None=QualitySwitch(Default=@n68, Epic=@n70)) {name=Textures + Details, nodeColor=<rgba 1 0.8203 0 1>}) {attributeGetTypes [#1, #4, #5]}
  n72 = ScalarParameter() {defaultValue=0.5, group=Waterline, parameterName=Waterline Value}
  n73 = Saturate(None=Divide(A=NamedRerouteUsage(), B=ScalarParameter() {defaultValue=300, group=Waterline, parameterName=Waterline Height Blr, sortPriority=3}))
  n74 = ScalarParameter() {defaultValue=0.25, group=Waterline, parameterName=Waterline Spec, sortPriority=1}
  n75 = SetMaterialAttributes!(
   MaterialAttributes=@n71.MaterialAttributes
   Base Color=Multiply(A=@n71.Base Color, B=LinearInterpolate(A=@n72, Alpha=@n73) {constA=0.15})
   Roughness=LinearInterpolate(A=ScalarParameter() {defaultValue=0.2, group=Waterline, parameterName=Waterline Roughness, sortPriority=1}, B=@n71.Roughness, Alpha=@n73) {constA=0.5}
   Specular=LinearInterpolate(A=Multiply(A=@n71.Specular, B=@n74), B=@n71.Specular, Alpha=@n73)
  ) {attributeSetTypes [#1, #4, #5]}
  n76 = NamedRerouteUsage()
  NamedRerouteDeclaration(None=QualitySwitch(Default=StaticSwitchParameter(True=@n75, False=@n76) {defaultValue=true, group=Interior, parameterName="Waterline "}, Low=@n76)) {name=Before Sand, nodeColor=<rgba 1 0 0.2109 1>}
  n77 = WorldPosition()
  n78 = Constant3Vector() {constant=<rgba 40607.9961 -67791.9922 2112 0>}
  n79 = Constant() {r=3500}
  n80 = SphereMask(A=@n77.XYZ, B=Constant3Vector() {constant=<rgba 40246.3164 -85753.2422 2668.5872 0>}, Radius=Constant() {r=3000}) {hardnessPercent=0}
  n81 = SphereMask(A=@n77.XYZ, B=Constant3Vector() {constant=<rgba 44144 -83208 2304.6726 0>}, Radius=Constant() {r=4500}) {hardnessPercent=0}
  n82 = Saturate(None=Add(A=SphereMask(A=@n77.XYZ, B=@n78, Radius=@n79) {hardnessPercent=0}, B=Add(A=@n80, B=@n81)))
  MaterialExpressionReroute_2 = Reroute(None=@n82)
  NamedRerouteDeclaration(None=@MaterialExpressionReroute_2.Output) {name=Light Masks, nodeColor=<rgba 1 0 0.3281 1>}
  n83 = MaterialFunctionCall(
   Diffuse Texture Object=@n61
   TextureSize=Multiply(A=ScalarParameter() {defaultValue=2, group=Rocks, parameterName=ScaleRocks}, B=Constant() {r=1500})
   NormalTextureObject=@n63
   World Space Normals=StaticBool()
   Export Float 4 Diffuse=StaticBool()
  ) {materialFunction=&WorldAlignedTextures_Complex}
  n84 = ComponentMask(None=@n83.XYZ Diffuse Texture) {b=true, g=true, r=true}
  n85 = ScalarParameter() {defaultValue=0.5, group=Rocks, parameterName=Rocks Spec, sortPriority=3}
  n86 = ScalarParameter() {defaultValue=0.6393, parameterName=R1}
  n87 = ScalarParameter() {defaultValue=2, parameterName=R2}
  n88 = ComponentMask(None=@n84) {r=true}
  NamedRerouteDeclaration(None=MakeMaterialAttributes!(
    BaseColor=@n84
    Metallic
    Specular=Multiply(A=@n84, B=@n85)
    Roughness=LinearInterpolate(A=@n86, B=@n87, Alpha=@n88)
    Anisotropy
    EmissiveColor
    Opacity
    OpacityMask
    Normal=@n83.Normal XYZ Texture
    Tangent
    WorldPositionOffset
    SubsurfaceColor
    ClearCoat
    ClearCoatRoughness
    AmbientOcclusion
    Refraction
    CustomizedUVs_0
    CustomizedUVs_1
    CustomizedUVs_2
    CustomizedUVs_3
    CustomizedUVs_4
    CustomizedUVs_5
    CustomizedUVs_6
    CustomizedUVs_7
    PixelDepthOffset
    ShadingModel
    Displacement
   )) {name=WS Rocks, nodeColor=<rgba 1 0.1172 0 1>}
  n89 = TextureObjectParameter() {
   group=Sand 2
   parameterName=Sand Specular 2
   samplerType=SAMPLERTYPE_Masks
   texture=/Game/Environments/Apollo/Sets/Terrain/Textures/T_Apollo_Gravel_S_01
  }
  n90 = TextureObjectParameter() {
   group=Sand 2
   parameterName=Sand Normal 2
   samplerType=SAMPLERTYPE_Normal
   texture=/Game/Environments/Apollo/Sets/Terrain/Textures/T_Apollo_Gravel_N_01
  }
  n91 = MaterialFunctionCall(
   Diffuse Texture Object=TextureObjectParameter() {group=Sand 2, parameterName=Sand Diffuse 2, texture=/Game/Environments/Apollo/Sets/Terrain/Textures/T_Apollo_Gravel_D_01}
   Spec Texture Object=@n89
   TextureSize=ScalarParameter() {defaultValue=812.1213, group=Sand 2, parameterName=Sand 2 WS Scale, sortPriority=1}
   NormalTextureObject=@n90
   World Space Normals=StaticBool()
   Export Float 4 Diffuse=StaticBool() {value=true}
  ) {materialFunction=&WorldAlignedTextures_Complex}
  n92 = MaterialFunctionCall(Float3=@n91.Z Spec) {materialFunction=&BreakOutFloat3Components}
  n93 = SetMaterialAttributes!(
   MaterialAttributes
   Base Color=Multiply(
    A=VectorParameter() {group=Sand 2, parameterName=ColorSand 2, defaultValue=<rgba 0.967 0.9321 0.8364 0.5>}.RGB
    B=ComponentMask(None=@n91.XYZ Diffuse Texture) {b=true, g=true, r=true}
   )
   Roughness=@n92.B
   Specular=@n92.B
   Normal=@n91.Normal Z Texture
  ) {attributeSetTypes [
    #1
    #4
    #5
    #2
   ]}
  n94 = BlendMaterialAttributes(
   A=NamedRerouteUsage()
   B=@n93
   Alpha=MaterialFunctionCall(Transition Phase=VertexColor().R, Height Texture=ComponentMask(None=@n91.XYZ Diffuse Texture) {a=true}) {materialFunction=&HeightLerp}.Lerp Alpha No Contrast
  )
  MaterialExpressionReroute_4 = Reroute(None=@n94)
  NamedRerouteDeclaration(None=@MaterialExpressionReroute_4.Output) {name=Sand Layer 2 Blend, nodeColor=<rgba 1 0 0.5625 1>}
  n95 = MaterialFunctionCall!(
   WSDeformationCenter=VectorParameter() {bUseCustomPrimitiveData=true, group=z_code, parameterName=WSDeformationCenter, primitiveDataIndex=20}.RGB
   Code_DestructionAmount=ScalarParameter() {group=z_code, parameterName=Code_DestructionAmount}
   Code_UseUvForDamage
   Code_UseHitEffects=StaticBoolParameter() {defaultValue=true, group=z_code, parameterName=Code_UseHitEffects}
   Code_Emissive
   WSDamageRadius=ScalarParameter() {bUseCustomPrimitiveData=true, defaultValue=1, group=z_code, parameterName=WSDamageRadius, primitiveDataIndex=26}
   CrackDamageShader
   Diffuse
   ScalarModDamageCracks
   ScalarOffsetforDamageCracks
   ExposeLerpFunctionForInvertingTheHitDistanceCalc
   InvertDamageFalloff
   CrackWorldPosition
   WorldPosition
   DistanceCullFadeAnimationOffset=ScalarParameter() {bUseCustomPrimitiveData=true, group=z_code, parameterName=DistanceCullFadeAnimationOffset, primitiveDataIndex=12}
  ) {materialFunction=/Game/Packages/jl_tests/Fortnite_HitEffects}
  NamedRerouteDeclaration(None=@n95.Hit And Shrink WPO) {name=Hit WPO, nodeColor=<rgba 0 1 0.9766 1>}
  n96 = NamedRerouteUsage()
  n97 = NamedRerouteUsage()
  n98 = TextureObjectParameter() {
   group=Sand
   parameterName=Sand Normal
   samplerType=SAMPLERTYPE_Normal
   texture=&T_Apollo_Sand_N_01
  }
  n99 = Add(
   A=Multiply(A=QualitySwitch(Default=@n96, Low=@n97), B=Constant3Vector() {constant=<rgba 0.5 0.5 0 0>})
   B=Multiply(
    A=TextureSample(Tex=@n98) {samplerSource=SSM_Wrap_WorldGroupSettings, samplerType=SAMPLERTYPE_Normal}.RGB
    B=Constant3Vector() {constant=<rgba 2 2 1 0>}
   )
  )
  n100 = SetMaterialAttributes(Base Color=NamedRerouteUsage(), Roughness=NamedRerouteUsage(), Specular=NamedRerouteUsage(), Normal=@n99) {attributeSetTypes [
    #1
    #4
    #5
    #2
   ]}
  n101 = SetMaterialAttributes(Base Color=NamedRerouteUsage(), Roughness=NamedRerouteUsage(), Specular=NamedRerouteUsage(), Normal=@n99) {attributeSetTypes [
    #1
    #4
    #5
    #2
   ]}
  n102 = VertexColor()
  n103 = NamedRerouteUsage()
  NamedRerouteDeclaration(None=BlendMaterialAttributes(
    A=NamedRerouteUsage()
    B=QualitySwitch(Default=@n100, Low=@n101, High=@n100, Medium=@n100)
    Alpha=MaterialFunctionCall(Transition Phase=@n102.B, Height Texture=@n103) {materialFunction=&HeightLerp}.Lerp Alpha No Contrast
   ) {vertexAttributeBlendType=UseA}) {name=Sand Mesh, nodeColor=<rgba 0.8516 1 0 1>}
  n104 = NamedRerouteUsage()
  n105 = ScalarParameter() {defaultValue=32, group=Sand, parameterName=Sand Water Height}
  NamedRerouteDeclaration(None=Saturate(None=Divide(A=@n104, B=@n105))) {name=Sand Waterline Mask, nodeColor=<rgba 0 0.8359 1 1>}
  NamedRerouteDeclaration(None=LinearInterpolate(
    A=Constant3Vector() {constant=<rgba 0 0 1 0>}
    B=NamedRerouteDeclaration(None=@n44.Normal Z Texture) {name="Sand  N Z", nodeColor=<rgba 1 0 0.2578 1>}
    Alpha=NamedRerouteUsage()
   )) {name=Flat N Waterline, nodeColor=<rgba 1 0.2578 0 1>}
  PinBase()
  PinBase!(Output=@n94) {reroutePins [Output {expression=&self:MaterialExpressionReroute_4}]}
  PinBase!()
  PinBase!(Output=@n18) {reroutePins [Output {expression=&self:MaterialExpressionReroute_1}]}
  PinBase!()
  PinBase!(Output=@n82) {reroutePins [Output {expression=&self:MaterialExpressionReroute_2}]}
  Reroute(None=@n95.Hit And Shrink WPO)
 ]
}
