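# Master translucent deferred decal material for the Artemis Celebration environment set, defaulting to confetti textures.
# Diffuse, specular mask and normal textures feed material attributes, with the mask's green channel shared by Metallic, Specular and Roughness.
# Color_Time_of_Day makes the diffuse emissive at night only, scaled by a parameter and Decal Tint.
# An Opacity parameter scales diffuse alpha, and MatLayerBlend_Tint applies the same tint to the attributes.
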Material /Game/Environments/Artemis/Sets/Celebration/Materials/Master_Materials/M_Celebration_Decal
graph {
 properties {
  bUseMaterialAttributes=true
  blendMode=BLEND_Translucent
  materialDomain=MD_DeferredDecal
  pixelDepthOffsetMode=PDOM_Legacy
  usedShadingModels=Default Lit
 }
 nodes [
  n1 = TextureSampleParameter2D() {parameterName=Diffuse, texture=/Game/Environments/Artemis/Sets/Celebration/Textures/T_Celebration_Confetti_D}
  n2 = TextureSampleParameter2D() {parameterName=Specular, samplerType=SAMPLERTYPE_Masks, texture=/Game/Environments/Artemis/Sets/Celebration/Textures/T_Celebration_Confetti_S}
  n3 = Constant()
  n4 = VectorParameter() {parameterName=Decal Tint, defaultValue=<rgba 1 1 1 1>}
  n5 = Multiply(
   A=MaterialFunctionCall(
    Color_Morning=@n3
    Color_Day=@n3
    Color_Evening=@n3
    Color_Night=ScalarParameter() {defaultValue=0.5, parameterName="Decal Emissive (Night)"}
   ) {materialFunction=&Color_Time_of_Day}.Result
   B=@n4.RGB
  )
  n6 = ScalarParameter() {defaultValue=1, parameterName=Opacity, sliderMax=1}
  n7 = TextureSampleParameter2D() {parameterName=Normal, samplerType=SAMPLERTYPE_Normal, texture=/Game/Environments/Artemis/Sets/Celebration/Textures/T_Celebration_Confetti_N}
  MaterialFunctionCall!(
   Material=MakeMaterialAttributes!(
    BaseColor=@n1.RGB
    Metallic=@n2.G
    Specular=@n2.G
    Roughness=@n2.G
    Anisotropy
    EmissiveColor=Multiply(A=@n1.RGB, B=@n5)
    Opacity=LinearInterpolate(B=@n1.A, Alpha=@n6)
    OpacityMask
    Normal=@n7.RGB
    Tangent
    WorldPositionOffset
    SubsurfaceColor
    ClearCoat
    ClearCoatRoughness
    AmbientOcclusion
    Refraction
    CustomizedUVs_0
    CustomizedUVs_1
    CustomizedUVs_2
    CustomizedUVs_3
    CustomizedUVs_4
    CustomizedUVs_5
    CustomizedUVs_6
    CustomizedUVs_7
    PixelDepthOffset
    ShadingModel
    Displacement
   )
   Tint=@n4.RGB
   TintMask
  ) {materialFunction=/Engine/Functions/MaterialLayerFunctions/MatLayerBlend_Tint}
 ]
}
