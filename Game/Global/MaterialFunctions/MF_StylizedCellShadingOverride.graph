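# Material function that overrides roughness for stylized cel shading, driven by the post-process volume user flags.
# Subtracting 100 from the view's user flags and saturating yields a blend that reaches full strength at 101.
# That blend lerps roughness toward fully rough (1).
# It serves both a material attributes path, changing only Roughness, and a separate scalar roughness input and output.
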MaterialFunction /Game/Global/MaterialFunctions/MF_StylizedCellShadingOverride
graph {nodes [
  n1 = GetMaterialAttributes(None=FunctionInput() {inputType=FunctionInput_MaterialAttributes}) {attributeGetTypes [#1]}
  n2 = Saturate(None=Subtract(A=ViewProperty() {property=MEVP_PostVolumeUserFlags}.Property) {constB=100})
  FunctionOutput(None=SetMaterialAttributes!(MaterialAttributes=@n1.MaterialAttributes, Roughness=LinearInterpolate(A=@n1.Roughness, Alpha=@n2)) {attributeSetTypes [#1]}) {outputName=MaterialOut}
  FunctionOutput(None=LinearInterpolate(A=FunctionInput() {inputName=Roughness, inputType=FunctionInput_Scalar, sortPriority=5}, Alpha=@n2)) {outputName=RoughnessOut, sortPriority=5}
 ]}
