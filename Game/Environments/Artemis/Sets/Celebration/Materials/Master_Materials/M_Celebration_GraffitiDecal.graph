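# Master deferred decal material for Celebration set graffiti, tinted by parameter or vertex color and finished with a stylized cel-shading override.
# One texture supplies color or channel masks, sampled by UV or world-aligned tiling faded with a spherical gradient, erosion and facing mask.
# Opacity, optionally scaled by vertex color, premultiplies the surface outputs for alpha compositing.
# Optional time-of-day emissive glows at night, and a paint texture varies roughness.
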Material /Game/Environments/Artemis/Sets/Celebration/Materials/Master_Materials/M_Celebration_GraffitiDecal
graph {
 properties {
  bUseMaterialAttributes=true
  bUsedWithNanite=true
  blendMode=BLEND_AlphaComposite
  materialDomain=MD_DeferredDecal
  pixelDepthOffsetMode=PDOM_Legacy
  usedShadingModels=Default Lit
 }
 nodes [
  n1 = TextureObjectParameter() {parameterName=Decal_Texture, texture=/Game/Environments/Artemis/Sets/Celebration/Textures/Graffiti/T_Celebration_Graffiti_Unicorn_D}
  n2 = MaterialFunctionCall(
   TextureObject=@n1
   TextureSize=ScalarParameter() {defaultValue=1024, parameterName=WS Decal Texture Size}
   Export Float 4=StaticBool() {value=true}
  ) {materialFunction=&WorldAlignedTexture}
  n3 = StaticSwitchParameter(True=ChannelMaskParameter(None=@n2.XYZ Texture) {parameterName=Mask Channel}.R, False=ComponentMask(None=@n2.XYZ Texture) {a=true}) {parameterName=Use Texture As Masks}
  n4 = SphereMask(
   A=TextureCoordinate()
   B=AppendVector(
    A=ScalarParameter() {defaultValue=0.5, parameterName=SphericalGradient_Center_U}
    B=ScalarParameter() {defaultValue=0.5, parameterName=SphericalGradient_Center_V}
   )
   Radius=ScalarParameter() {defaultValue=0.5, parameterName=SphericalGradientRadius}
   Hardness=ScalarParameter() {defaultValue=0.25, parameterName=SphericalGradientHardness}
  )
  n5 = Multiply(
   A=Subtract(A=Multiply(A=@n4, B=@n3), B=ScalarParameter() {parameterName=Erosion})
   B=MaterialFunctionCall() {materialFunction=/Game/Weapons/Prototype/SprayGun/Materials/MF_DecalFacingMask}.Result
  )
  n6 = TextureSample(Tex=@n1)
  n7 = ChannelMaskParameter(None=@n6.RGBA) {parameterName=Mask Channel}
  n8 = LinearInterpolate(
   B=StaticSwitchParameter(True=Multiply(A=@n3, B=@n5), False=StaticSwitchParameter(True=@n7.R, False=@n6.A) {parameterName=Use Texture As Masks}) {parameterName=WS Decal Tiling}
   Alpha=ScalarParameter() {defaultValue=1, parameterName=Opacity, sliderMax=1}
  )
  n9 = StaticSwitchParameter(True=Multiply(A=VertexColor(), B=@n8), False=@n8) {parameterName=Use Vertex Color Opacity Masking}
  n10 = StaticSwitchParameter(
   True=Constant() {r=1}
   False=StaticSwitchParameter(True=ComponentMask(None=@n2.XYZ Texture) {b=true, g=true, r=true}, False=@n6.RGB) {parameterName=WS Decal Tiling}
  ) {parameterName=Use Texture As Masks}
  n11 = ScalarParameter() {defaultValue=0.3, parameterName=Roughness Min}
  n12 = ScalarParameter() {defaultValue=0.55, parameterName=Roughness Max}
  n13 = TextureSample() {texture=/Game/Environments/Artemis/Sets/Celebration/Textures/CleanPaint/T_Celebration_Clean_Paint_Background_S}
  n14 = VectorParameter() {parameterName=Decal Tint, defaultValue=<rgba 1 1 1 1>}
  n15 = Constant()
  n16 = Constant() {r=0.3}
  n17 = Multiply(
   A=@n14.RGB
   B=StaticSwitchParameter(
    True=MaterialFunctionCall(Color_Morning=@n15, Color_Day=@n15, Color_Evening=@n15, Color_Night=@n16) {materialFunction=&Color_Time_of_Day}.Result
    False=Constant()
   ) {defaultValue=true, parameterName=Emissive at night?}
  )
  n18 = MakeMaterialAttributes!(
   BaseColor=Multiply(A=@n9, B=@n10)
   Metallic=Multiply(A=@n9, B=Constant())
   Specular=Multiply(A=@n9, B=ScalarParameter() {defaultValue=0.5, parameterName=Specular})
   Roughness=Multiply(A=@n9, B=LinearInterpolate(A=@n11, B=@n12, Alpha=@n13.B))
   Anisotropy
   EmissiveColor=Multiply(A=@n9, B=Multiply(A=@n17, B=@n10))
   Opacity=@n9
   OpacityMask
   Normal
   Tangent
   WorldPositionOffset
   SubsurfaceColor
   ClearCoat
   ClearCoatRoughness
   AmbientOcclusion
   Refraction
   CustomizedUVs_0
   CustomizedUVs_1
   CustomizedUVs_2
   CustomizedUVs_3
   CustomizedUVs_4
   CustomizedUVs_5
   CustomizedUVs_6
   CustomizedUVs_7
   PixelDepthOffset
   ShadingModel
   Displacement
  )
  n19 = MaterialFunctionCall!(Material=@n18, Tint=@n14.RGB, TintMask) {materialFunction=/Engine/Functions/MaterialLayerFunctions/MatLayerBlend_Tint}
  n20 = MaterialFunctionCall(Material=@n18, Tint=VertexColor()) {materialFunction=&MatLayerBlend_Tint}
  MaterialFunctionCall!(
   In=StaticSwitchParameter(
    True=@n19.Blended Material
    False=StaticSwitchParameter(True=@n20.Blended Material, False=@n19.Blended Material) {parameterName=Use Vertex Colors for Tinting}
   ) {parameterName=Use Vertex Color Opacity Masking}
   Roughness
  ) {materialFunction=/Game/Global/MaterialFunctions/MF_StylizedCellShadingOverride}
 ]
}
